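# One-shot mono MetaSound for the theme park stage event countdown, sent to the stage fireworks speaker bus.
# Start plays immediately while one-second trigger repeats fire Tick, then hand off to Final via a trigger counter.
# Interpolators raise pitch partway through and ramp the Tick/Final mix gain across the countdown.
# OnFinished fires from a counter on Final's second play; the trailing delay node's output is unconnected.
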MetaSoundSource /ThemeParkStageEvent/Audio/MSS_ThemeParkStageEvent_Countdown
inputs [
 Final {type=WaveAsset}
 Start {type=WaveAsset}
 Tick {type=WaveAsset}
 UE.Source.OnPlay {type=Trigger}
]
interfaces [UE.OutputFormat.Mono:1.0, UE.Source.OneShot:1.0, UE.Source:1.1]
outputs [
 "UE.OutputFormat.Mono.Audio:0" {type=Audio}
 UE.Source.OneShot.OnFinished {type=Trigger}
]
pages [Default {
  input_defaults {
   Final=/ThemeParkStageEvent/Audio/SoundWaves/ThemeParkStageEvent_Countdown_Final
   Start=/ThemeParkStageEvent/Audio/SoundWaves/ThemeParkStageEvent_Countdown_Start
   Tick=/ThemeParkStageEvent/Audio/SoundWaves/ThemeParkStageEvent_Countdown_Tick
   UE.Source.OnPlay=null
  }
  nodes [
   n1 = UE.TriggerRepeat(Start=^UE.Source.OnPlay) {also {defaults {Num Repeats=10, Period=1}}}
   TriggerCounter_2 = UE.Trigger Counter(In=@n1.RepeatOut) {also {defaults {Reset Count=5}}}
   n2 = UE.InterpTo:Audio(Target=Value.Value:Float(Set=@TriggerCounter_2.On Reset) {also {defaults {Init Value=0, Target Value=3}}}.Output Value) {also {defaults {Interp Time=5}}}
   WavePlayer_0 = UE.Wave Player:Mono(
    Wave Asset=^Tick
    Pitch Shift=@n2.Value
    Play=UE.TriggerRepeat(Start=^UE.Source.OnPlay) {also {defaults {Num Repeats=7, Period=1}}}.RepeatOut
   ) {also {version [1,1]}}
   TriggerCounter_0 = UE.Trigger Counter(In=@n1.RepeatOut) {also {defaults {Reset Count=7}}}
   WavePlayer_2 = UE.Wave Player:Mono(
    Wave Asset=^Final
    Pitch Shift=@n2.Value
    Play=UE.TriggerRepeat(Start=@TriggerCounter_0.On Reset) {also {defaults {Num Repeats=3, Period=1}}}.RepeatOut
   ) {also {version [1,1]}}
   AudioMixerAudioMixerMono2_0 = "AudioMixer.Audio Mixer (Mono, 2)"(In 0=@WavePlayer_0.Out Mono, In 1=@WavePlayer_2.Out Mono)
   WavePlayer_1 = UE.Wave Player:Mono(Wave Asset=^Start, Play=^UE.Source.OnPlay) {also {version [1,1]}}
   AudioMixerAudioMixerMono2_1 = "AudioMixer.Audio Mixer (Mono, 2)"(
    In 1=@AudioMixerAudioMixerMono2_0.Out
    Gain 1=UE.InterpTo:Audio(Target=Value.Value:Float(Set=^UE.Source.OnPlay) {also {defaults {Init Value=0, Target Value=2}}}.Output Value) {also {defaults {Interp Time=10}}}.Value
    In 0=@WavePlayer_1.Out Mono
   )
   TriggerCounter_1 = UE.Trigger Counter(In=@WavePlayer_2.On Play) {also {defaults {Reset Count=2}}}
   TriggerDelay_0 = UE.Trigger Delay(In=@TriggerCounter_1.On Reset) {also {defaults {Delay Time=4}}}
  ]
  links [
   {dst="@UE.OutputFormat.Mono.Audio:0", src=AudioMixerAudioMixerMono2_1.Out}
   {dst="@UE.Source.OneShot.OnFinished", src=TriggerCounter_1.On Reset}
  ]
 }]
properties {
 assetImportData=&self:AssetImportData
 attenuationSettings=/Game/Athena/Sounds/Events/Papaya/Papaya_TheaterStage_Event_Att
 bEnableBaseSubmix=false
 bEnableSubmixSends=false
 bOverrideConcurrency=true
 maxDistance=27048
 priority=100
 sampleRate=48000
 soundClassObject=&Fort_Mission_Ingame_Important_PSM
 virtualizationMode=PlayWhenSilent
 volume=0.125
 busSends [{audioBus=/ThemeParkStageEvent/Audio/Music/AB_ThemeParkStage_Speakers_Fireworks, sendLevel=1}]
 concurrencyOverrides {maxCount=1, resolutionRule=StopOldest}
}
