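# Sound cue for the crowd sound during the fireworks of the Theme Park stage event, in the Fort_Events sound class.
# A single wave player feeds a mixer at the root, with no randomization or looping node.
# It uses the shared crowd attenuation asset; the inline attenuation values are stored without the override flag enabled.
# It keeps playing when silent, overrides concurrency, and has its debug flag left on.
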SoundCue /ThemeParkStageEvent/Audio/SC_ThemeParkStageEvent_Fireworks_Crowd
root=Mixer_0
nodes [
 {id=Mixer_0, children [WavePlayer_0], properties {inputVolume [1]}}
 {id=WavePlayer_0, properties {soundWaveAssetPtr=/ThemeParkStageEvent/Audio/SoundWaves/ThemeParkStageEvent_Fireworks_Crowd}}
]
properties {
 attenuationSettings=&/ATT_ThemeParkStageEvent_Crowd
 bDebug=true
 bOverrideConcurrency=true
 maxDistance=6000
 soundClassObject=&Fort_Events
 virtualizationMode=PlayWhenSilent
 volumeMultiplier=1
 attenuationOverrides {
  bAttenuateWithLPF=true
  bEnableListenerFocus=true
  falloffDistance=1500
  lPFFrequencyAtMin=400
  lPFRadiusMax=4000
  lPFRadiusMin=3500
  nonFocusVolumeAttenuation=0.9
  nonSpatializedRadiusEnd=1000
  nonSpatializedRadiusMode=Direct2D
  nonSpatializedRadiusStart=1000
  attenuationShapeExtents {x=2500}
 }
}
