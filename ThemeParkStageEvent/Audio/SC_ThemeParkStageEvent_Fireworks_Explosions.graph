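# Sound cue for the fireworks explosion audio in the Theme Park stage event.
# A single fireworks wave player, with no looping or randomization, feeds a Mixer that boosts its input volume.
# It uses a dedicated fireworks attenuation asset, belongs to the Fort_Events sound class, and keeps playing when silent through virtualization.
# The debug flag is left enabled on the cue.
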SoundCue /ThemeParkStageEvent/Audio/SC_ThemeParkStageEvent_Fireworks_Explosions
root=Mixer_0
nodes [
 {id=Mixer_0, children [WavePlayer_0], properties {inputVolume [2]}}
 {id=WavePlayer_0, properties {soundWaveAssetPtr=/ThemeParkStageEvent/Audio/SoundWaves/ThemeParkStageEvent_Fireworks_Fireworks}}
]
properties {
 attenuationSettings=&/ATT_ThemeParkStageEvent_Fireworks
 bDebug=true
 maxDistance=26000
 soundClassObject=&Fort_Events
 virtualizationMode=PlayWhenSilent
 volumeMultiplier=1
}
