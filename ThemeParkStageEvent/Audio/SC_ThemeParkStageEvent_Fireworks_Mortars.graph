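# Sound cue for the mortar fireworks in the Theme Park Stage Event, routed to the Fort_Events sound class.
# A single wave player feeds a mixer that boosts its input volume.
# Spatial falloff comes from a shared mortar attenuation asset, and inline attenuation values are stored without their override switched on.
# The cue keeps playing when silent, overrides concurrency and has its debug flag enabled.
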SoundCue /ThemeParkStageEvent/Audio/SC_ThemeParkStageEvent_Fireworks_Mortars
root=Mixer_0
nodes [
 {id=Mixer_0, children [WavePlayer_0], properties {inputVolume [2]}}
 {id=WavePlayer_0, properties {soundWaveAssetPtr=/ThemeParkStageEvent/Audio/SoundWaves/ThemeParkStageEvent_Fireworks_Mortars}}
]
properties {
 attenuationSettings=&/ATT_ThemeParkStageEvent_Mortars
 bDebug=true
 bOverrideConcurrency=true
 maxDistance=5500
 soundClassObject=&Fort_Events
 virtualizationMode=PlayWhenSilent
 volumeMultiplier=1
 attenuationOverrides {
  bEnableListenerFocus=true
  bEnableReverbSend=false
  bEnableSubmixSends=true
  distanceAlgorithm=Logarithmic
  falloffDistance=2000
  lPFFrequencyAtMin=40
  lPFRadiusMax=1600
  lPFRadiusMin=600
  nonFocusVolumeAttenuation=0.8
  nonSpatializedRadiusEnd=1500
  nonSpatializedRadiusMode=Direct2D
  nonSpatializedRadiusStart=1000
  attenuationShapeExtents {x=3500}
 }
}
